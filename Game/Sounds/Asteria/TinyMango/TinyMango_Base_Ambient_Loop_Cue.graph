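# Sound cue for the looping base ambient sound of TinyMango, classed as an Athena ambient prop sound.
# A looping wave player feeds an Enveloper root node, whose volume curve fades the sound in from silence.
# Spatial falloff comes from a dedicated attenuation asset, with a boosted volume multiplier on the cue.
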SoundCue /Game/Sounds/Asteria/TinyMango/TinyMango_Base_Ambient_Loop_Cue
root=Enveloper_0
nodes [
 {
  id=Enveloper_0
  children [WavePlayer_0]
  properties {pitchCurve {editorCurveData {keys [{time=0}]}}, volumeCurve {editorCurveData {keys [<at 0 0>,<at 1 1>]}}}
 }
 {id=WavePlayer_0, properties {bLooping=true, soundWaveAssetPtr=&/TinyMango_Base_Amb_Loop}}
]
properties {
 attenuationSettings=&/TinyMango_Base_Ambient_Att
 maxDistance=3750
 soundClassObject=&Fort_Ambient_Props_Athena
 volumeMultiplier=1.504
}
