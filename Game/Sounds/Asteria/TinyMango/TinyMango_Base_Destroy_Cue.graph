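# Sound cue for the destruction sound of the TinyMango base in the Asteria sound set.
# A single wave player feeds the output directly with one destroy wave, with no randomization or modulation.
# The cue boosts volume, routes to the Fort_Impacts sound class, and uses the attenuation asset named for the base's ambient sound rather than a destroy-specific one.
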SoundCue /Game/Sounds/Asteria/TinyMango/TinyMango_Base_Destroy_Cue
root=WavePlayer_0
nodes [{id=WavePlayer_0, properties {soundWaveAssetPtr=&/TinyMango_Base_Destroy_01}}]
properties {
 attenuationSettings=&/TinyMango_Base_Ambient_Att
 maxDistance=3750
 soundClassObject=&Fort_Impacts
 volumeMultiplier=2
}
